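# Material function for the SleepyTime player body that fakes cheap toon eye and brow expressions by swapping mesh pieces through World Position Offset.
# Pose scalars are scaled by thresholds and rounded, switching target vertex-color values for eyes (red: happy, squint, blink) and brows (green: up, glabella).
# VertexColorMask results, plus a head green mask, are combined with Max; the inverse collapses vertices onto the object position, added to incoming WPO.
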MaterialFunction /Game/Characters/Player/Male/Medium/Bodies/M_MED_SleepyTime/Materials/MF_CheapToonEyes
graph {nodes [
  n1 = GetMaterialAttributes(None=FunctionInput() {inputName=Material, inputType=FunctionInput_MaterialAttributes}) {attributeGetTypes [#1]}
  n2 = WorldPosition()
  n3 = ObjectPositionWS()
  n4 = Constant() {r=1}
  n5 = VertexColor()
  n6 = ChannelMaskParameter(None=@n5) {group=Banana, parameterName=EyeVertexColorChannel Red}
  n7 = ScalarParameter() {defaultValue=0.375, group=Banana, parameterName=Threshold_squint_inner_pose}
  n8 = Round(None=Multiply(A=ScalarParameter() {group=Banana_Poses, parameterName=R_squint_inner_pose}, B=Divide(B=@n7) {constA=0.5}))
  n9 = Multiply(
   A=ScalarParameter() {group=Banana_Poses, parameterName=R_blink_pose}
   B=Divide(B=ScalarParameter() {defaultValue=0.1, group=Banana, parameterName=Threshold_R_blink_pose}) {constA=0.5}
  )
  n10 = LinearInterpolate(
   A=LinearInterpolate(
    A=ScalarParameter() {defaultValue=0.3, group=Banana, parameterName="BananaHappyPoseValue(R)"}
    B=ScalarParameter() {defaultValue=0.5, group=Banana, parameterName="BananaSquintPoseValue(R)"}
    Alpha=Clamp(None=@n8)
   )
   B=ScalarParameter() {defaultValue=0.6, group=Banana, parameterName="BananaBlinkPoseValue(R)"}
   Alpha=Clamp(None=Round(None=@n9))
  )
  n11 = ChannelMaskParameter(None=@n5) {group=Banana, maskChannel=Green, parameterName=EyeBrowVertexColorChannel Green, defaultValue=<rgba 0 1 0 0>}
  n12 = Constant() {r=0.1}
  n13 = ScalarParameter() {defaultValue=0.4, group=Banana, parameterName="BananaBrowUpValue(R)"}
  n14 = ScalarParameter() {group=Banana_Poses, parameterName=R_brow_up_pose}
  n15 = Divide(B=ScalarParameter() {defaultValue=0.1, group=Banana, parameterName=Threshold_R_brow_up_pose}) {constA=0.5}
  n16 = Clamp(None=Round(None=Multiply(A=@n14, B=@n15)))
  n17 = ScalarParameter() {defaultValue=0.1, group=Banana, parameterName=Threshold_Glabella_down}
  n18 = Round(None=Multiply(A=ScalarParameter() {group=Banana_Poses, parameterName=C_glabella_down_pose}, B=Divide(B=@n17) {constA=0.5}))
  n19 = Multiply(
   A=ScalarParameter() {group=Banana_Poses, parameterName=C_glabella_up_pose}
   B=Divide(B=ScalarParameter() {defaultValue=0.1, group=Banana, parameterName=Threshold_Glabella_up}) {constA=0.5}
  )
  n20 = LinearInterpolate(
   A=LinearInterpolate(
    A=LinearInterpolate(A=@n12, B=@n13, Alpha=@n16)
    B=ScalarParameter() {defaultValue=0.2, group=Banana, parameterName="GlabellaDownValue(G)"}
    Alpha=Clamp(None=@n18)
   )
   B=@n12
   Alpha=Clamp(None=Round(None=@n19))
  )
  n21 = Max(
   A=Max(
    A=MaterialFunctionCall!(Hide Element Toggle=@n4, VertexColorChannel=@n6.R, Vertex Value to hide=@n10) {materialFunction=/Game/Packages/Fortress_SharedMaterials/Base_Material_Functions/VertexColorMask}.Result
    B=MaterialFunctionCall(Hide Element Toggle=@n4, VertexColorChannel=@n11.R, Vertex Value to hide=@n20) {materialFunction=&VertexColorMask}.Result
   )
   B=MaterialFunctionCall(
    Hide Element Toggle=Constant() {r=1}
    VertexColorChannel=@n5.G
    Vertex Value to hide=ScalarParameter() {defaultValue=1, parameterName=HeadGreenValue}
   ) {materialFunction=&VertexColorMask}.Result
  )
  n22 = Multiply(A=Multiply(A=Subtract(A=@n2.XYZ, B=@n3)) {constB=-1}, B=Saturate(None=OneMinus(None=@n21)))
  FunctionOutput(None=SetMaterialAttributes!(MaterialAttributes=@n1.MaterialAttributes, WorldPositionOffset=Add(A=@n1.WorldPositionOffset, B=@n22)) {attributeSetTypes [#1]})
  ScalarParameter() {defaultValue=-0.1, parameterName=RB_GlitchLineSpeed}
  TextureCoordinate() {coordinateIndex=1}
  Time()
 ]}
